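# Translucent, unlit, two-sided Niagara mesh and ribbon material in the M_MED_Crisis body FX, drawing spline-thickened, noise-distorted strands in particle color.
# Two panned dissolve-texture samples drive a view-transformed world position offset, sine-masked along UV V and tuned by dynamic parameters.
# A third panned sample varies SplineThicken's base and tip widths, whose offset adds to the wobble.
# Particle color feeds emissive adjustments; particle alpha passes through depth fade.
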Material /BRCosmetics/Characters/Player/Male/Medium/Bodies/M_MED_Crisis/FX/M_MED_Crisis
graph {
 properties {
  bUsedWithNiagaraMeshParticles=true
  bUsedWithNiagaraRibbons=true
  blendMode=BLEND_Translucent
  pixelDepthOffsetMode=PDOM_Legacy
  shadingModel=MSM_Unlit
  twoSided=true
  usedShadingModels=Unlit
 }
 nodes [
  n1 = VectorParameter() {parameterName=Param_1, defaultValue=<rgba 1 1 0 1>}
  n2 = VertexColor()
  n3 = Multiply(
   A=Multiply(
    A=DynamicParameter() {paramNames [Panner_Speed, WPO_Intensity, WPO_UVtiling, Param4]}.WPO_UVtiling
    B=TextureCoordinate() {uTiling=0, vTiling=1.5}
   )
   B=@n2.R
  )
  n4 = Multiply(A=Time(), B=DynamicParameter() {paramNames [Panner_Speed, WPO_Intensity, WPO_UVtiling, Param4]}.Panner_Speed)
  n5 = TextureSample(UVs=Panner(Coordinate=Add(A=@n3, B=@n2.R), Time=@n4) {bFractionalPart=true, speedX=0.1, speedY=0.8}) {texture=/Game/CharClasses/Outlander/Gadgets/AntiMaterialCharge/Materials/T_AntiMaterialCharge_Dissolve_01}
  n6 = Multiply(
   A=TextureCoordinate() {uTiling=0, vTiling=1.2}
   B=DynamicParameter() {paramNames [Panner_Speed, WPO_Intensity, WPO_UVtiling, Param4]}.WPO_UVtiling
  )
  n7 = VertexColor()
  n8 = TextureSample(UVs=Panner(Coordinate=Add(A=@n6, B=@n7.R), Time=@n4) {bFractionalPart=true, speedY=0.8}) {texture=&T_AntiMaterialCharge_Dissolve_01}
  n9 = LinearInterpolate(Alpha=LinearInterpolate(A=Power(Base=@n5.R) {constExponent=4}, B=LinearInterpolate(Alpha=@n8.R) {constA=-1})) {constA=-1.5, constB=1.6}
  n10 = DynamicParameter() {paramNames [Panner_Speed, WPO_Intensity, WPO_UVtiling, Param4]}
  n11 = Transform(None=Multiply(
    A=MaterialFunctionCall(X=@n9, Y=@n9, Z=@n9) {materialFunction=&MakeFloat3}.Result
    B=@n10.WPO_Intensity
   ) {constB=0.8}) {transformType=TRANSFORM_View}
  n12 = ComponentMask(None=TextureCoordinate()) {g=true}
  n13 = ScalarParameter() {defaultValue=6, parameterName=Base}
  n14 = Panner(Coordinate=TextureCoordinate() {uTiling=0, vTiling=0.65}) {speedX=2, speedY=2}
  n15 = LinearInterpolate(Alpha=Power(Base=TextureSample(UVs=@n14) {texture=&T_AntiMaterialCharge_Dissolve_01}.R)) {constA=-0.65, constB=1.75}
  n16 = ComponentMask(None=TextureCoordinate()) {g=true}
  n17 = Multiply(A=@n15, B=OneMinus(None=@n16))
  n18 = Multiply(A=ScalarParameter() {defaultValue=3, parameterName=Tip}, B=@n10.Param4)
  n19 = Multiply(A=@n15, B=@n16)
  Add(
   A=Multiply(A=Multiply(A=@n1.RGB, B=@n11), B=Sine(None=@n12) {period=2})
   B=MaterialFunctionCall(WidthBase=Multiply(A=@n13, B=@n17), WidthTip=Multiply(A=@n18, B=@n19)) {materialFunction=&SplineThicken}.WorldPositionOffset
  )
  n20 = ParticleColor()
  DepthFade(Opacity=@n20.A) {fadeDistanceDefault=12}
  MaterialFunctionCall(Emissive=@n20.RGB) {materialFunction=&MF_FX_EmissiveAdjustments}
 ]
}
