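# Niagara system for the M_MED_Crisis player body cosmetic, adding magenta energy glow and lightning.
# Energy draws glow sprites and a light, while HeadToShoulder, HeadToShoulderWide and HeadUp spawn fading lightning meshes, all following the user-supplied skeletal mesh.
# User.Variant scales their spawn counts and rates, and bisHeadHidden drives their renderer visibility tags.
# Per-limb bolt emitters and a partitioned Lightining emitter, which indexes per-bone offsets from arrays, are disabled.
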NiagaraSystem /BRCosmetics/Characters/Player/Male/Medium/Bodies/M_MED_Crisis/FX/NS_MED_Crisis
dependencies {
 usedDataInterfaces [
  &NiagaraDataInterfaceCurve
  &NiagaraDataInterfaceColorCurve
  &NiagaraDataInterfaceVectorCurve
  &NiagaraDataInterfaceVector2DCurve
  &NiagaraDataInterfaceMeshRendererInfo
  &NiagaraDataInterfaceArrayFloat3
  &NiagaraDataInterfaceArrayColor
 ]
 usedDynamicInputs [&MaskFloatBySpawnGroup, &SelectLinearColorFromArray]
 usedModules [
  &self:SystemSpawnScript.SystemScriptSource.SystemScriptGraph.NiagaraNodeAssignment_0.SetVariables_#1
  &self:SystemSpawnScript.SystemScriptSource.SystemScriptGraph.NiagaraNodeAssignment_1.SetVariables_#2
  &SystemState
 ]
}
emitters {
 Energy {
  emitterUpdateScript [
   &EmitterState(Life Cycle Mode=::NewEnumerator1, Loop Behavior=::NewEnumerator1, Loop Duration Mode=::NewEnumerator1, UseLoopDelay=false)
   &SpawnRate(Use Spawn Probability=false) {enabled=false}
   &SpawnBurst_Instantaneous(
    Spawn Count=&Multiply_Int(A=1, B=&MakeIntFromFloat(Float=$User.Variant))
    Use Spawn Probability=false
    UseLoopCountLimit=false
   )
  ]
  particleSpawnScript [&InitializeParticle(
    Color=<rgba 1 0.0272 0.8296 0.15>
    Sprite Size Mode=::NewEnumerator3
    Uniform Sprite Size=40
    UsePositionOffset=false
    Write Material Random=true
   )]
  particleUpdateScript [
   &ParticleState(Kill Particles When Lifetime Has Elapsed=false, Let Infinitely Lived Particles Die When Emitter Deactivates=true)
   &SkeletalMeshLocation(Bone Sampling Mode=::NewEnumerator4, MaskSpawnGroup=false, Sampled Position Offset=<12 5 0>, Skeletal Mesh=$User.New Skeletal Mesh)
   &FN_Radial_Material(Base Sprite Tile V=2, Sprite Extrusion Speed=0.5)
   &ScaleMeshSize(Scale Factor=&VectorFromFloat(Value=&FloatFromCurve(FloatCurve=DataInterface() {propertyValues {Curve {keys [{interpMode=RCIM_Cubic, tangentMode=RCTM_User, time=0, value=0}, {interpMode=RCIM_Cubic, tangentMode=RCTM_User, time=1, value=1}]}}})))
   /SetVariables(Particles.Renderer2_Color:LinearColor=<rgba 0.5713 0 1 1>) {isSetParameters=true}
   /SetVariables(Particles.Renderer2_Size:Vector2f=&Multiply_Vector2D(A=$Particles.SpriteSize, B=&Vector2DFromFloat(Value=1.75))) {isSetParameters=true}
   /SetVariables(Particles.Renderer3_Size:Vector2f=&Multiply_Vector2D_ByFloat(Float=0.6, Vector2D=$Particles.SpriteSize)) {isSetParameters=true}
   &FN_Light_Attributes(
    Color=<rgba 1 0 0.7523 1>
    Create Color Attribute=true
    Create Exponent Attribute=true
    Create Radius Attribute=true
    Exponent=6
    Radius=100
    Scale Color=3
   )
   /SetVariables(Particles.VisibilityTag=&MakeCustomIntFromBool(Bool=$User.bisHeadHidden, True INT=1, False INT=0)) {isSetParameters=true, module=SetVariables_#3}
  ]
  rendererClasses [&NiagaraSpriteRendererProperties, &NiagaraLightRendererProperties]
  renderers [
   {rendererClass=&NiagaraSpriteRendererProperties}
   {rendererClass=&NiagaraSpriteRendererProperties, rendererIndex=1}
   {rendererClass=&NiagaraLightRendererProperties, rendererIndex=2}
   {rendererClass=&NiagaraSpriteRendererProperties, rendererIndex=3}
  ]
 }
 Hand_L1 {
  bEnabled=false
  emitterUpdateScript [
   &EmitterState(Life Cycle Mode=::NewEnumerator1, Loop Delay=0.3, UseLoopDelay=true)
   &SpawnRate(SpawnRate=1.2, Use Spawn Probability=false)
  ]
  particleSpawnScript [
   &InitializeParticle(
    Color=$System.LightningColor
    Lifetime Max=1
    Lifetime Min=0.2
    Lifetime Mode=::NewEnumerator1
    Mesh Scale Mode=::NewEnumerator4
    UsePositionOffset=false
    Write Material Random=true
   )
   &InitialMeshOrientation(Rotation=<0 -0.25 0>, Rotation Coordinate Space=::NewEnumerator2, Use Orientation Vector=true, Use Rotation Vector=true)
   &DynamicMaterialParameters(
    Index 0 Param 1=3
    Index 0 Param 2=1.1
    Index 0 Param 3=0.55
    Index 0 Param 4=1.4
    Module.Param0WriteEnabled=true
    Module.Param1WriteEnabled=true
    Module.Param2WriteEnabled=true
    Module.Param3WriteEnabled=true
   )
  ]
  particleUpdateScript [
   &ParticleState()
   &SkeletalMeshLocation(
    Bone Sampling Mode=::NewEnumerator0
    "Interpolate Mesh Sampling (Update)"=true
    MaskSpawnGroup=false
    OverrideRandomSeed=false
    Sampled Orientation Coordinate Space=::Local
    Sampled Position Offset=<9 2 4>
    Skeletal Mesh=DataInterface() {propertyValues {
      DefaultMesh=null
      ExcludeBoneName=root
      OptionalComponentClass=null
      PreviewMesh=/BRCosmetics/Characters/Player/Male/Medium/Bodies/M_MED_Crisis/Meshes/M_MED_Crisis
      SkinningMode=SkinOnTheFly
      SoftSourceActor=null
      SourceMode=Default
      UvSetIndex=0
      WholeMeshLOD=-1
      bExcludeBone=false
      bReadDeformedGeometry=true
      bRequireCurrentFrameData=true
      ComponentTags []
      FilteredBones [upperarm_l]
      FilteredSockets []
      MeshUserParameter {parameter {name=None, typeDefHandle {registeredTypeIndex=1}, varData []}}
      SamplingRegions []
     }}
   )
   &ScaleColor(
    Scale Alpha=&FloatFromCurve(FloatCurve=DataInterface() {propertyValues {Curve {keys [
         {arriveTangent=4.237, interpMode=RCIM_Cubic, leaveTangent=4.237, tangentMode=RCTM_User, time=0, value=0}
         {arriveTangent=0.1123, interpMode=RCIM_Cubic, leaveTangent=0.1123, tangentMode=RCTM_User, time=0.5, value=1}
         {arriveTangent=-4.5416, interpMode=RCIM_Cubic, leaveTangent=-4.5416, tangentMode=RCTM_User, time=1, value=0}
        ]}}})
    ScaleA=true
    ScaleRGB=true
   )
   &FN_Light_Attributes(
    Color=<rgba 1 0.1307 0.7967 1>
    Create Color Attribute=true
    Create Exponent Attribute=true
    Create Position Attribute=true
    Create Radius Attribute=true
    Exponent=1
    Position=&Add_Vector(A=$StackContext.SkeletalMeshLocation.SampledPosition, B=<15 6 -7>)
    Radius=12
    Scale Color=6
   )
  ]
  rendererClasses [&NiagaraMeshRendererProperties, &NiagaraLightRendererProperties]
  renderers [
   {rendererClass=&NiagaraMeshRendererProperties, rendererIndex=0}
   {rendererClass=&NiagaraLightRendererProperties, rendererIndex=1}
  ]
 }
 Hand_L1001 {
  bEnabled=false
  emitterUpdateScript [
   &EmitterState(Life Cycle Mode=::NewEnumerator1, Loop Delay=0.3, UseLoopDelay=true)
   &SpawnRate(SpawnRate=1.2, Use Spawn Probability=false)
  ]
  particleSpawnScript [
   &InitializeParticle(
    Color=$System.LightningColor
    Lifetime Max=1
    Lifetime Min=0.1
    Lifetime Mode=::NewEnumerator1
    Mesh Scale Mode=::NewEnumerator4
    UsePositionOffset=false
    Write Material Random=true
   )
   &InitialMeshOrientation(Rotation=<-0.05 0.2 0>, Rotation Coordinate Space=::NewEnumerator2, Use Orientation Vector=true, Use Rotation Vector=true)
   &DynamicMaterialParameters(
    Index 0 Param 1=3
    Index 0 Param 2=1.1
    Index 0 Param 3=0.55
    Index 0 Param 4=1.4
    Module.Param0WriteEnabled=true
    Module.Param1WriteEnabled=true
    Module.Param2WriteEnabled=true
    Module.Param3WriteEnabled=true
   )
  ]
  particleUpdateScript [
   &ParticleState()
   &SkeletalMeshLocation(
    Bone Sampling Mode=::NewEnumerator0
    "Interpolate Mesh Sampling (Update)"=true
    MaskSpawnGroup=false
    OverrideRandomSeed=false
    Sampled Orientation Coordinate Space=::Local
    Sampled Position Offset=<-4 0 -12>
    Skeletal Mesh=DataInterface() {propertyValues {
      DefaultMesh=null
      ExcludeBoneName=root
      OptionalComponentClass=null
      PreviewMesh=&M_MED_Crisis
      SkinningMode=SkinOnTheFly
      SoftSourceActor=null
      SourceMode=Default
      UvSetIndex=0
      WholeMeshLOD=-1
      bExcludeBone=false
      bReadDeformedGeometry=true
      bRequireCurrentFrameData=true
      ComponentTags []
      FilteredBones [head_offset]
      FilteredSockets []
      MeshUserParameter {parameter {name=None, typeDefHandle {registeredTypeIndex=1}, varData []}}
      SamplingRegions []
     }}
   )
   &ScaleColor(
    Scale Alpha=&FloatFromCurve(FloatCurve=DataInterface() {propertyValues {Curve {keys [
         {arriveTangent=4.237, interpMode=RCIM_Cubic, leaveTangent=4.237, tangentMode=RCTM_User, time=0, value=0}
         {arriveTangent=0.1123, interpMode=RCIM_Cubic, leaveTangent=0.1123, tangentMode=RCTM_User, time=0.5, value=1}
         {arriveTangent=-4.5416, interpMode=RCIM_Cubic, leaveTangent=-4.5416, tangentMode=RCTM_User, time=1, value=0}
        ]}}})
    ScaleA=true
    ScaleRGB=true
   )
   &FN_Light_Attributes(
    Color=<rgba 1 0.1307 0.7967 1>
    Create Color Attribute=true
    Create Exponent Attribute=true
    Create Position Attribute=true
    Create Radius Attribute=true
    Exponent=2
    Position=&Add_Vector(A=$StackContext.SkeletalMeshLocation.SampledPosition, B=<10 10 0>)
    Radius=16
    Scale Color=20
   )
   /SetVariables(Particles.VisibilityTag=&MakeCustomIntFromBool(Bool=$User.bisHeadHidden, True INT=1, False INT=0)) {isSetParameters=true}
  ]
  rendererClasses [&NiagaraMeshRendererProperties, &NiagaraLightRendererProperties]
  renderers [
   {rendererClass=&NiagaraMeshRendererProperties, rendererIndex=0}
   {rendererClass=&NiagaraLightRendererProperties, rendererIndex=1}
  ]
 }
 Hand_L2 {
  bEnabled=false
  emitterUpdateScript [
   &EmitterState(Life Cycle Mode=::NewEnumerator1, Loop Delay=0.3, UseLoopDelay=true)
   &SpawnRate(SpawnRate=1.2, Use Spawn Probability=false)
  ]
  particleSpawnScript [
   &InitializeParticle(
    Color=$System.LightningColor
    Lifetime Max=1
    Lifetime Min=0.1
    Lifetime Mode=::NewEnumerator1
    UsePositionOffset=false
    Write Material Random=true
   )
   &InitialMeshOrientation(Rotation=<0 -0.25 0>, Rotation Coordinate Space=::NewEnumerator2, Use Orientation Vector=true, Use Rotation Vector=true)
   &DynamicMaterialParameters(
    Index 0 Param 1=&RandomRangeFloat(Minimum=3, Maximum=3)
    Index 0 Param 2=1
    Index 0 Param 3=0.6
    Index 0 Param 4=0.5
    Module.Param0WriteEnabled=true
    Module.Param1WriteEnabled=true
    Module.Param2WriteEnabled=true
    Module.Param3WriteEnabled=true
   )
  ]
  particleUpdateScript [
   &ParticleState()
   &SkeletalMeshLocation(
    Bone Sampling Mode=::NewEnumerator0
    "Interpolate Mesh Sampling (Update)"=true
    MaskSpawnGroup=false
    OverrideRandomSeed=false
    Sampled Orientation Coordinate Space=::Local
    Sampled Position Offset=<5 2 1>
    Skeletal Mesh=DataInterface() {propertyValues {
      DefaultMesh=null
      ExcludeBoneName=root
      OptionalComponentClass=null
      PreviewMesh=&M_MED_Crisis
      SkinningMode=SkinOnTheFly
      SoftSourceActor=null
      SourceMode=Default
      UvSetIndex=0
      WholeMeshLOD=-1
      bExcludeBone=false
      bReadDeformedGeometry=true
      bRequireCurrentFrameData=true
      ComponentTags []
      FilteredBones [lowerarm_l]
      FilteredSockets []
      MeshUserParameter {parameter {name=None, typeDefHandle {registeredTypeIndex=1}, varData []}}
      SamplingRegions []
     }}
   )
   &ScaleColor(
    Scale Alpha=&FloatFromCurve(FloatCurve=DataInterface() {propertyValues {Curve {keys [
         {arriveTangent=4.237, interpMode=RCIM_Cubic, leaveTangent=4.237, tangentMode=RCTM_User, time=0, value=0}
         {arriveTangent=0.1123, interpMode=RCIM_Cubic, leaveTangent=0.1123, tangentMode=RCTM_User, time=0.5, value=1}
         {arriveTangent=-4.5416, interpMode=RCIM_Cubic, leaveTangent=-4.5416, tangentMode=RCTM_User, time=1, value=0}
        ]}}})
    ScaleA=true
    ScaleRGB=true
   )
   &FN_Light_Attributes(
    Color=<rgba 1 0.1307 0.7967 1>
    Create Color Attribute=true
    Create Exponent Attribute=true
    Create Position Attribute=true
    Create Radius Attribute=true
    Exponent=1
    Position=&Add_Vector(A=$StackContext.SkeletalMeshLocation.SampledPosition, B=<-10 9 0>)
    Radius=12
    Scale Color=6
   )
   /SetVariables(Particles.VisibilityTag=&MakeCustomIntFromBool(Bool=$User.bisHandHidden, True INT=1, False INT=0)) {isSetParameters=true}
  ]
  rendererClasses [&NiagaraMeshRendererProperties, &NiagaraLightRendererProperties]
  renderers [
   {rendererClass=&NiagaraMeshRendererProperties, rendererIndex=0}
   {rendererClass=&NiagaraLightRendererProperties, rendererIndex=1}
  ]
 }
 Hand_R1001 {
  bEnabled=false
  emitterUpdateScript [
   &EmitterState(
    Life Cycle Mode=::NewEnumerator1
    Loop Delay=&RandomRangeFloat(Minimum=0, Maximum=0.5)
    Loop Duration=&RandomRangeFloat(Minimum=0.4, Maximum=1)
    UseLoopDelay=true
   )
   &SpawnRate(SpawnRate=1.2, Use Spawn Probability=false)
  ]
  particleSpawnScript [
   &InitializeParticle(
    Color=$System.LightningColor
    Lifetime Max=1
    Lifetime Min=0.1
    Lifetime Mode=::NewEnumerator1
    Mesh Scale Mode=::NewEnumerator4
    UsePositionOffset=false
    Write Material Random=true
   )
   &InitialMeshOrientation(Rotation=<0 -0.25 0.1>, Rotation Coordinate Space=::NewEnumerator2, Use Orientation Vector=true, Use Rotation Vector=true)
   &DynamicMaterialParameters(
    Index 0 Param 1=&RandomRangeFloat(Minimum=3, Maximum=3)
    Index 0 Param 2=&RandomRangeFloat(Minimum=1, Maximum=1)
    Index 0 Param 3=0.7
    Index 0 Param 4=0.8
    Module.Param0WriteEnabled=true
    Module.Param1WriteEnabled=true
    Module.Param2WriteEnabled=true
    Module.Param3WriteEnabled=true
   )
   &CameraOffset()
  ]
  particleUpdateScript [
   &ParticleState()
   &SkeletalMeshLocation(
    Bone Sampling Mode=::NewEnumerator0
    "Interpolate Mesh Sampling (Update)"=true
    MaskSpawnGroup=false
    OverrideRandomSeed=false
    Sampled Orientation Coordinate Space=::Local
    Sampled Position Offset=<-10 0 3>
    Skeletal Mesh=DataInterface() {propertyValues {
      DefaultMesh=null
      ExcludeBoneName=root
      OptionalComponentClass=null
      PreviewMesh=&M_MED_Crisis
      SkinningMode=SkinOnTheFly
      SoftSourceActor=null
      SourceMode=Default
      UvSetIndex=0
      WholeMeshLOD=-1
      bExcludeBone=false
      bReadDeformedGeometry=true
      bRequireCurrentFrameData=true
      ComponentTags []
      FilteredBones [lowerarm_r]
      FilteredSockets []
      MeshUserParameter {parameter {name=None, typeDefHandle {registeredTypeIndex=1}, varData []}}
      SamplingRegions []
     }}
   )
   &ScaleColor(
    Scale Alpha=&FloatFromCurve(FloatCurve=DataInterface() {propertyValues {Curve {keys [
         {arriveTangent=4.237, interpMode=RCIM_Cubic, leaveTangent=4.237, tangentMode=RCTM_User, time=0, value=0}
         {arriveTangent=0.1123, interpMode=RCIM_Cubic, leaveTangent=0.1123, tangentMode=RCTM_User, time=0.5, value=1}
         {arriveTangent=-4.5416, interpMode=RCIM_Cubic, leaveTangent=-4.5416, tangentMode=RCTM_User, time=1, value=0}
        ]}}})
    ScaleA=true
    ScaleRGB=true
   )
   /SetVariables(Particles.VisibilityTag=&MakeCustomIntFromBool(Bool=$User.bisHandHidden, True INT=1, False INT=0)) {isSetParameters=true}
  ]
  rendererClasses [&NiagaraMeshRendererProperties]
  renderers [{rendererClass=&NiagaraMeshRendererProperties, rendererIndex=0}]
 }
 HeadToShoulder {
  emitterUpdateScript [
   &EmitterState()
   &SpawnRate(SpawnRate=&Multiply_Float(A=4, B=$User.Variant), Use Spawn Probability=false)
   &SpawnBurst_Instantaneous(Spawn Count=4, Use Spawn Probability=false, UseLoopCountLimit=false) {enabled=false}
   &SpawnBurst_Instantaneous(Spawn Count=4, Spawn Time=1.5123, Use Spawn Probability=false, UseLoopCountLimit=false) {enabled=false, module=SpawnBurst_Instantaneous001}
   &SpawnBurst_Instantaneous(Spawn Count=4, Spawn Time=4.1222, Use Spawn Probability=false, UseLoopCountLimit=false) {enabled=false, module=SpawnBurst_Instantaneous002}
  ]
  particleSpawnScript [
   &InitializeParticle(
    Color=$System.LightningColor
    Lifetime Max=1
    Lifetime Min=0.1
    Lifetime Mode=::NewEnumerator1
    Mesh Scale Mode=::NewEnumerator1
    Mesh Uniform Scale Max=1
    Mesh Uniform Scale Min=0.7
    UsePositionOffset=false
    Write Material Random=true
   )
   &InitialMeshOrientation(
    Orientation Vector=<0 0 1>
    Rotation=&RandomRangeVector(Minimum=<-0.04 -0.07 0>, Maximum=<0 0.07 0>)
    Use Orientation Vector=true
    Use Rotation Vector=true
   )
   &InitialMeshOrientation(
    Rotation=&RandomRangeVector(Minimum=<-1 0 0>, Maximum=<1 0 0>)
    Use Orientation Vector=false
    Use Rotation Vector=true
   ) {enabled=false, module=InitialMeshOrientation001}
   &DynamicMaterialParameters(
    Index 0 Param 1=3
    Index 0 Param 2=1.1
    Index 0 Param 3=0.55
    Index 0 Param 4=&RandomRangeFloat(Minimum=0.2, Maximum=0.7)
    Module.Param0WriteEnabled=true
    Module.Param1WriteEnabled=true
    Module.Param2WriteEnabled=true
    Module.Param3WriteEnabled=true
   )
   /SetVariables(Particles.MeshIndex=&RandomRangeInt(Minimum=0, Maximum=2)) {isSetParameters=true, module=SetVariables_#4}
  ]
  particleUpdateScript [
   &ParticleState()
   &SkeletalMeshLocation(Bone Sampling Mode=::NewEnumerator4, MaskSpawnGroup=false, Sampled Position Offset=<12 0 0>, Skeletal Mesh=$User.New Skeletal Mesh)
   &ScaleColor(
    Scale Alpha=&FloatFromCurve(FloatCurve=DataInterface() {propertyValues {Curve {keys [
         {arriveTangent=4.237, interpMode=RCIM_Cubic, leaveTangent=4.237, tangentMode=RCTM_User, time=0, value=0}
         {arriveTangent=0.1123, interpMode=RCIM_Cubic, leaveTangent=0.1123, tangentMode=RCTM_User, time=0.5, value=1}
         {arriveTangent=-4.5416, interpMode=RCIM_Cubic, leaveTangent=-4.5416, tangentMode=RCTM_User, time=1, value=0}
        ]}}})
    ScaleA=true
    ScaleRGB=true
   )
   &FN_Light_Attributes(
    Color=<rgba 1 0.1307 0.7967 1>
    Create Color Attribute=true
    Create Exponent Attribute=true
    Create Position Attribute=true
    Create Radius Attribute=true
    Exponent=2
    Position=&Add_Vector(A=$StackContext.SkeletalMeshLocation.SampledPosition, B=<10 10 0>)
    Radius=16
    Scale Color=20
   ) {enabled=false}
   &UpdateMeshOrientation(Rotation Vector=<0 1 0>) {enabled=false}
   &ScaleMeshSize(Scale Factor=&VectorFromCurve(VectorCurve=DataInterface() {propertyValues {XCurve {keys [<at 0 0>,<at 1 1>]}, YCurve {keys [<at 0 0>,<at 1 1>]}, ZCurve {keys [<at 0 0>,<at 1 1>]}}})) {enabled=false}
   /SetVariables(Particles.VisibilityTag=&MakeCustomIntFromBool(Bool=$User.bisHeadHidden, True INT=1, False INT=0)) {isSetParameters=true, module=SetVariables_#3}
  ]
  rendererClasses [&NiagaraMeshRendererProperties]
  renderers [{rendererClass=&NiagaraMeshRendererProperties, rendererIndex=0}]
 }
 HeadToShoulderWide {
  emitterUpdateScript [
   &EmitterState()
   &SpawnRate(SpawnRate=&Multiply_Float(A=2, B=$User.Variant), Use Spawn Probability=false)
   &SpawnBurst_Instantaneous(Spawn Count=4, Use Spawn Probability=false, UseLoopCountLimit=false) {enabled=false}
  ]
  particleSpawnScript [
   &FN_SetSpawnGroup(Number of Spawn Groups=2)
   &InitializeParticle(
    Color=$System.LightningColor
    Lifetime Max=1
    Lifetime Min=0.1
    Lifetime Mode=::NewEnumerator1
    Mesh Scale=&MaskVectorBySpawnGroup(Masked Vector=<0.7 0.7 0.8>, Passthrough Vector=<-0.8 -0.8 -1>)
    Mesh Scale Mode=::NewEnumerator4
    UsePositionOffset=false
    Write Material Random=true
   )
   &InitialMeshOrientation(
    Rotation=&MaskVectorBySpawnGroup(Masked Vector=<-0.15 -0.1 0.2>, Passthrough Vector=<0.1 -0.1 0.45>)
    Use Orientation Vector=true
    Use Rotation Vector=true
   )
   &InitialMeshOrientation(
    Rotation=&RandomRangeVector(Minimum=<-1 0 0>, Maximum=<1 0 0>)
    Use Orientation Vector=false
    Use Rotation Vector=true
   ) {enabled=false, module=InitialMeshOrientation001}
   &DynamicMaterialParameters(
    Index 0 Param 1=3
    Index 0 Param 2=1.1
    Index 0 Param 3=0.55
    Index 0 Param 4=0.6
    Module.Param0WriteEnabled=true
    Module.Param1WriteEnabled=true
    Module.Param2WriteEnabled=true
    Module.Param3WriteEnabled=true
   )
   /SetVariables(Particles.MeshIndex=&RandomRangeInt(Minimum=0, Maximum=3)) {enabled=false, isSetParameters=true, module=SetVariables_#4}
  ]
  particleUpdateScript [
   &ParticleState()
   &SkeletalMeshLocation(Bone Sampling Mode=::NewEnumerator4, MaskSpawnGroup=false, Sampled Position Offset=<12 0 0>, Skeletal Mesh=$User.New Skeletal Mesh)
   &ScaleColor(
    Scale Alpha=&FloatFromCurve(FloatCurve=DataInterface() {propertyValues {Curve {keys [
         {arriveTangent=4.237, interpMode=RCIM_Cubic, leaveTangent=4.237, tangentMode=RCTM_User, time=0, value=0}
         {arriveTangent=0.1123, interpMode=RCIM_Cubic, leaveTangent=0.1123, tangentMode=RCTM_User, time=0.5, value=1}
         {arriveTangent=-4.5416, interpMode=RCIM_Cubic, leaveTangent=-4.5416, tangentMode=RCTM_User, time=1, value=0}
        ]}}})
    ScaleA=true
    ScaleRGB=true
   )
   &FN_Light_Attributes(
    Color=<rgba 1 0.1307 0.7967 1>
    Create Color Attribute=true
    Create Exponent Attribute=true
    Create Position Attribute=true
    Create Radius Attribute=true
    Exponent=2
    Position=&Add_Vector(A=$StackContext.SkeletalMeshLocation.SampledPosition, B=<10 10 0>)
    Radius=16
    Scale Color=20
   ) {enabled=false}
   &UpdateMeshOrientation(Rotation Vector=<0 1 0>) {enabled=false}
   &ScaleMeshSize(Scale Factor=&VectorFromCurve(VectorCurve=DataInterface() {propertyValues {XCurve {keys [<at 0 0>,<at 1 1>]}, YCurve {keys [<at 0 0>,<at 1 1>]}, ZCurve {keys [<at 0 0>,<at 1 1>]}}}))
   /SetVariables(Particles.VisibilityTag=&MakeCustomIntFromBool(Bool=$User.bisHeadHidden, True INT=1, False INT=0)) {isSetParameters=true, module=SetVariables_#3}
  ]
  rendererClasses [&NiagaraMeshRendererProperties]
  renderers [{rendererClass=&NiagaraMeshRendererProperties, rendererIndex=0}]
 }
 HeadUp {
  emitterUpdateScript [
   &EmitterState()
   &SpawnRate(SpawnRate=&Multiply_Float(A=1, B=$User.Variant), Use Spawn Probability=false)
   &SpawnBurst_Instantaneous(Spawn Count=4, Use Spawn Probability=false, UseLoopCountLimit=false) {enabled=false}
  ]
  particleSpawnScript [
   &InitializeParticle(
    Color=$System.LightningColor
    Lifetime Max=1
    Lifetime Min=0.1
    Lifetime Mode=::NewEnumerator1
    Mesh Scale Mode=::NewEnumerator3
    Mesh Uniform Scale=&RandomRangeFloat(Minimum=0.8, Maximum=1)
    UsePositionOffset=false
    Write Material Random=true
   )
   &InitialMeshOrientation(
    Orientation Vector=<0 0 1>
    Rotation=&RandomRangeVector(Minimum=<0 0.3 0>, Maximum=<0 0.7 0>)
    Use Orientation Vector=true
    Use Rotation Vector=true
   )
   &InitialMeshOrientation(
    Rotation=&RandomRangeVector(Minimum=<-1 0 0>, Maximum=<1 0 0>)
    Use Orientation Vector=false
    Use Rotation Vector=true
   ) {enabled=false, module=InitialMeshOrientation001}
   &DynamicMaterialParameters(
    Index 0 Param 1=3
    Index 0 Param 2=1.1
    Index 0 Param 3=0.55
    Index 0 Param 4=0.6
    Module.Param0WriteEnabled=true
    Module.Param1WriteEnabled=true
    Module.Param2WriteEnabled=true
    Module.Param3WriteEnabled=true
   )
   /SetVariables(Particles.MeshIndex=&RandomRangeInt(Minimum=0, Maximum=3)) {isSetParameters=true, module=SetVariables_#4}
  ]
  particleUpdateScript [
   &ParticleState()
   &SkeletalMeshLocation(Bone Sampling Mode=::NewEnumerator4, MaskSpawnGroup=false, Sampled Position Offset=<12 5 0>, Skeletal Mesh=$User.New Skeletal Mesh)
   &ScaleColor(
    Scale Alpha=&FloatFromCurve(FloatCurve=DataInterface() {propertyValues {Curve {keys [
         {arriveTangent=4.237, interpMode=RCIM_Cubic, leaveTangent=4.237, tangentMode=RCTM_User, time=0, value=0}
         {arriveTangent=0.1123, interpMode=RCIM_Cubic, leaveTangent=0.1123, tangentMode=RCTM_User, time=0.5, value=1}
         {arriveTangent=-4.5416, interpMode=RCIM_Cubic, leaveTangent=-4.5416, tangentMode=RCTM_User, time=1, value=0}
        ]}}})
    ScaleA=true
    ScaleRGB=true
   )
   &FN_Light_Attributes(
    Color=<rgba 1 0.1307 0.7967 1>
    Create Color Attribute=true
    Create Exponent Attribute=true
    Create Position Attribute=true
    Create Radius Attribute=true
    Exponent=2
    Position=&Add_Vector(A=$StackContext.SkeletalMeshLocation.SampledPosition, B=<10 10 0>)
    Radius=16
    Scale Color=20
   ) {enabled=false}
   &UpdateMeshOrientation(Rotation Vector=<0 1 0>) {enabled=false}
   &ScaleMeshSize(Scale Factor=&VectorFromCurve(VectorCurve=DataInterface() {propertyValues {XCurve {keys [<at 0 0>,<at 1 1>]}, YCurve {keys [<at 0 0>,<at 1 1>]}, ZCurve {keys [<at 0 0>,<at 1 1>]}}}))
   /SetVariables(Particles.VisibilityTag=&MakeCustomIntFromBool(Bool=$User.bisHeadHidden, True INT=1, False INT=0)) {isSetParameters=true, module=SetVariables_#3}
  ]
  rendererClasses [&NiagaraMeshRendererProperties]
  renderers [{rendererClass=&NiagaraMeshRendererProperties, rendererIndex=0}]
 }
 Leg_L1 {
  bEnabled=false
  emitterUpdateScript [
   &EmitterState(
    Life Cycle Mode=::NewEnumerator1
    Loop Delay=&RandomRangeFloat(Minimum=0.2, Maximum=0.4)
    Loop Duration=&RandomRangeFloat(Minimum=0.4, Maximum=1)
    UseLoopDelay=true
   )
   &SpawnRate(SpawnRate=1.2, Use Spawn Probability=false)
  ]
  particleSpawnScript [
   &InitializeParticle(
    Color=$System.LightningColor
    Lifetime Max=1
    Lifetime Min=0.1
    Lifetime Mode=::NewEnumerator1
    Mesh Scale Mode=::NewEnumerator4
    UsePositionOffset=false
    Write Material Random=true
   )
   &InitialMeshOrientation(Rotation=<0 -0.25 0>, Rotation Coordinate Space=::NewEnumerator2, Use Orientation Vector=true, Use Rotation Vector=true)
   &DynamicMaterialParameters(
    Index 0 Param 1=&RandomRangeFloat(Minimum=-3, Maximum=-3)
    Index 0 Param 2=&RandomRangeFloat(Minimum=1, Maximum=1)
    Index 0 Param 3=0.6
    Index 0 Param 4=1.2
    Module.Param0WriteEnabled=true
    Module.Param1WriteEnabled=true
    Module.Param2WriteEnabled=true
    Module.Param3WriteEnabled=true
   )
  ]
  particleUpdateScript [
   &ParticleState()
   &SkeletalMeshLocation(
    Bone Sampling Mode=::NewEnumerator0
    "Interpolate Mesh Sampling (Update)"=true
    MaskSpawnGroup=false
    OverrideRandomSeed=false
    Sampled Orientation Coordinate Space=::Local
    Sampled Position Offset=<-18 -6 0>
    Skeletal Mesh=DataInterface() {propertyValues {
      DefaultMesh=null
      ExcludeBoneName=root
      OptionalComponentClass=null
      PreviewMesh=&M_MED_Crisis
      SkinningMode=SkinOnTheFly
      SoftSourceActor=null
      SourceMode=Default
      UvSetIndex=0
      WholeMeshLOD=-1
      bExcludeBone=false
      bReadDeformedGeometry=true
      bRequireCurrentFrameData=true
      ComponentTags []
      FilteredBones [thigh_l]
      FilteredSockets []
      MeshUserParameter {parameter {name=None, typeDefHandle {registeredTypeIndex=1}, varData []}}
      SamplingRegions []
     }}
   )
   &ScaleColor(
    Scale Alpha=&FloatFromCurve(FloatCurve=DataInterface() {propertyValues {Curve {keys [
         {arriveTangent=4.237, interpMode=RCIM_Cubic, leaveTangent=4.237, tangentMode=RCTM_User, time=0, value=0}
         {arriveTangent=0.1123, interpMode=RCIM_Cubic, leaveTangent=0.1123, tangentMode=RCTM_User, time=0.5, value=1}
         {arriveTangent=-4.5416, interpMode=RCIM_Cubic, leaveTangent=-4.5416, tangentMode=RCTM_User, time=1, value=0}
        ]}}})
    ScaleA=true
    ScaleRGB=true
   )
   &FN_Light_Attributes(
    Color=<rgba 1 0.1307 0.7967 1>
    Create Color Attribute=true
    Create Exponent Attribute=true
    Create Position Attribute=true
    Create Radius Attribute=true
    Exponent=1
    Position=&Add_Vector(A=$StackContext.SkeletalMeshLocation.SampledPosition, B=<5 0 -24>)
    Radius=15
    Scale Color=10
   )
  ]
  rendererClasses [&NiagaraMeshRendererProperties, &NiagaraLightRendererProperties]
  renderers [
   {rendererClass=&NiagaraMeshRendererProperties, rendererIndex=0}
   {rendererClass=&NiagaraLightRendererProperties, rendererIndex=1}
  ]
 }
 Leg_L2 {
  bEnabled=false
  emitterUpdateScript [
   &EmitterState(
    Life Cycle Mode=::NewEnumerator1
    Loop Delay=&RandomRangeFloat(Minimum=0.2, Maximum=0.6)
    Loop Duration=&RandomRangeFloat(Minimum=0.4, Maximum=1)
    UseLoopDelay=true
   )
   &SpawnRate(SpawnRate=1.2, Use Spawn Probability=false)
  ]
  particleSpawnScript [
   &InitializeParticle(
    Color=$System.LightningColor
    Lifetime Max=1
    Lifetime Min=0.1
    Lifetime Mode=::NewEnumerator1
    Mesh Scale=<1 1 0.7>
    Mesh Scale Mode=::NewEnumerator4
    UsePositionOffset=false
    Write Material Random=true
   )
   &InitialMeshOrientation(Rotation=<0 0.25 0>, Rotation Coordinate Space=::NewEnumerator2, Use Orientation Vector=true, Use Rotation Vector=true)
   &DynamicMaterialParameters(
    Index 0 Param 1=&RandomRangeFloat(Minimum=3, Maximum=3)
    Index 0 Param 2=&RandomRangeFloat(Minimum=1.4, Maximum=1.4)
    Index 0 Param 3=0.48
    Index 0 Param 4=0.5
    Module.Param0WriteEnabled=true
    Module.Param1WriteEnabled=true
    Module.Param2WriteEnabled=true
    Module.Param3WriteEnabled=true
   )
  ]
  particleUpdateScript [
   &ParticleState()
   &SkeletalMeshLocation(
    Bone Sampling Mode=::NewEnumerator0
    "Interpolate Mesh Sampling (Update)"=true
    MaskSpawnGroup=false
    OverrideRandomSeed=false
    Sampled Orientation Coordinate Space=::Local
    Sampled Position Offset=<-20 0 -4>
    Skeletal Mesh=DataInterface() {propertyValues {
      DefaultMesh=null
      ExcludeBoneName=root
      OptionalComponentClass=null
      PreviewMesh=&M_MED_Crisis
      SkinningMode=SkinOnTheFly
      SoftSourceActor=null
      SourceMode=Default
      UvSetIndex=0
      WholeMeshLOD=-1
      bExcludeBone=false
      bReadDeformedGeometry=true
      bRequireCurrentFrameData=true
      ComponentTags []
      FilteredBones [calf_l]
      FilteredSockets []
      MeshUserParameter {parameter {name=None, typeDefHandle {registeredTypeIndex=1}, varData []}}
      SamplingRegions []
     }}
   )
   &ScaleColor(
    Scale Alpha=&FloatFromCurve(FloatCurve=DataInterface() {propertyValues {Curve {keys [
         {arriveTangent=4.237, interpMode=RCIM_Cubic, leaveTangent=4.237, tangentMode=RCTM_User, time=0, value=0}
         {arriveTangent=0.1123, interpMode=RCIM_Cubic, leaveTangent=0.1123, tangentMode=RCTM_User, time=0.5, value=1}
         {arriveTangent=-4.5416, interpMode=RCIM_Cubic, leaveTangent=-4.5416, tangentMode=RCTM_User, time=1, value=0}
        ]}}})
    ScaleA=true
    ScaleRGB=true
   )
   &FN_Light_Attributes(
    Color=<rgba 1 0.1307 0.7967 1>
    Create Color Attribute=true
    Create Exponent Attribute=true
    Create Position Attribute=true
    Create Radius Attribute=true
    Enable Particle Light Enable=true
    Exponent=1
    Light Enabled:NiagaraBool=&Boolean_NotOperation(Bool=$User.bisFootHidden)
    Position=&Add_Vector(A=$StackContext.SkeletalMeshLocation.SampledPosition, B=<15 0 -20>)
    Radius=28
    Scale Color=20
   )
   /SetVariables(Particles.VisibilityTag=&MakeCustomIntFromBool(Bool=$User.bisFootHidden, True INT=1, False INT=0)) {isSetParameters=true, module=SetVariables_#5}
  ]
  rendererClasses [&NiagaraMeshRendererProperties, &NiagaraLightRendererProperties]
  renderers [
   {rendererClass=&NiagaraMeshRendererProperties, rendererIndex=0}
   {rendererClass=&NiagaraLightRendererProperties, rendererIndex=1}
  ]
 }
 Leg_R {
  bEnabled=false
  emitterUpdateScript [
   &EmitterState(
    Life Cycle Mode=::NewEnumerator1
    Loop Delay=&RandomRangeFloat(Minimum=0, Maximum=0.8)
    UseLoopDelay=true
   )
   &SpawnRate(SpawnRate=1.2, Use Spawn Probability=false)
  ]
  particleSpawnScript [
   &InitializeParticle(
    Color=$System.LightningColor
    Lifetime Max=1
    Lifetime Min=0.1
    Lifetime Mode=::NewEnumerator1
    Mesh Scale=<1 1 0.8>
    Mesh Scale Mode=::NewEnumerator4
    UsePositionOffset=false
    Write Material Random=true
   )
   &InitialMeshOrientation(Rotation=<0 -0.24 0>, Rotation Coordinate Space=::NewEnumerator2, Use Orientation Vector=true, Use Rotation Vector=true)
   &DynamicMaterialParameters(
    Index 0 Param 1=&RandomRangeFloat(Minimum=3, Maximum=3)
    Index 0 Param 2=1.2
    Index 0 Param 3=0.6
    Index 0 Param 4=0.46
    Module.Param0WriteEnabled=true
    Module.Param1WriteEnabled=true
    Module.Param2WriteEnabled=true
    Module.Param3WriteEnabled=true
   )
   &CameraOffset()
   &InitialMeshOrientation(Rotation=<0.3 -0.05 0>, Use Orientation Vector=true, Use Rotation Vector=true) {module=InitialMeshOrientation001}
  ]
  particleUpdateScript [
   &ParticleState()
   &SkeletalMeshLocation(
    Bone Sampling Mode=::NewEnumerator0
    "Interpolate Mesh Sampling (Update)"=true
    MaskSpawnGroup=false
    OverrideRandomSeed=false
    Sampled Orientation Coordinate Space=::Local
    Sampled Position Offset=<26 -3 2>
    Skeletal Mesh=DataInterface() {propertyValues {
      DefaultMesh=null
      ExcludeBoneName=root
      OptionalComponentClass=null
      PreviewMesh=&M_MED_Crisis
      SkinningMode=SkinOnTheFly
      SoftSourceActor=null
      SourceMode=Default
      UvSetIndex=0
      WholeMeshLOD=-1
      bExcludeBone=false
      bReadDeformedGeometry=true
      bRequireCurrentFrameData=true
      ComponentTags []
      FilteredBones [calf_r]
      FilteredSockets []
      MeshUserParameter {parameter {name=None, typeDefHandle {registeredTypeIndex=1}, varData []}}
      SamplingRegions []
     }}
   )
   &ScaleColor(
    Scale Alpha=&FloatFromCurve(FloatCurve=DataInterface() {propertyValues {Curve {keys [
         {arriveTangent=4.237, interpMode=RCIM_Cubic, leaveTangent=4.237, tangentMode=RCTM_User, time=0, value=0}
         {arriveTangent=0.1123, interpMode=RCIM_Cubic, leaveTangent=0.1123, tangentMode=RCTM_User, time=0.5, value=1}
         {arriveTangent=-4.5416, interpMode=RCIM_Cubic, leaveTangent=-4.5416, tangentMode=RCTM_User, time=1, value=0}
        ]}}})
    ScaleA=true
    ScaleRGB=true
   )
   &FN_Light_Attributes(
    Color=<rgba 1 0.1307 0.7967 1>
    Create Color Attribute=true
    Create Exponent Attribute=true
    Create Position Attribute=true
    Create Radius Attribute=true
    Enable Particle Light Enable=true
    Exponent=1
    Light Enabled:NiagaraBool=&Boolean_NotOperation(Bool=$User.bisFootHidden)
    Position=&Add_Vector(A=$StackContext.SkeletalMeshLocation.SampledPosition, B=<-5 0 -20>)
    Radius=22
    Scale Color=10
   )
   /SetVariables(Particles.VisibilityTag=&MakeCustomIntFromBool(Bool=$User.bisFootHidden, True INT=1, False INT=0)) {isSetParameters=true, module=SetVariables_#5}
  ]
  rendererClasses [&NiagaraMeshRendererProperties, &NiagaraLightRendererProperties]
  renderers [
   {rendererClass=&NiagaraMeshRendererProperties, rendererIndex=0}
   {rendererClass=&NiagaraLightRendererProperties, rendererIndex=1}
  ]
 }
 Leg_R1 {
  bEnabled=false
  emitterUpdateScript [
   &EmitterState(
    Life Cycle Mode=::NewEnumerator1
    Loop Delay=&RandomRangeFloat(Minimum=0, Maximum=0.8)
    UseLoopDelay=true
   )
   &SpawnRate(SpawnRate=1.2, Use Spawn Probability=false)
  ]
  particleSpawnScript [
   &InitializeParticle(
    Color=$System.LightningColor
    Lifetime Max=1
    Lifetime Min=0.1
    Lifetime Mode=::NewEnumerator1
    Mesh Scale=<1 1 0.89>
    Mesh Scale Mode=::NewEnumerator4
    UsePositionOffset=false
    Write Material Random=true
   )
   &InitialMeshOrientation(Rotation=<0 -0.2 0>, Rotation Coordinate Space=::NewEnumerator2, Use Orientation Vector=true, Use Rotation Vector=true)
   &DynamicMaterialParameters(
    Index 0 Param 1=&RandomRangeFloat(Minimum=3, Maximum=3)
    Index 0 Param 2=1.2
    Index 0 Param 3=0.5
    Index 0 Param 4=1.1
    Module.Param0WriteEnabled=true
    Module.Param1WriteEnabled=true
    Module.Param2WriteEnabled=true
    Module.Param3WriteEnabled=true
   )
   &InitialMeshOrientation(Rotation=<0.2 0 0>, Use Orientation Vector=true, Use Rotation Vector=true) {module=InitialMeshOrientation001}
  ]
  particleUpdateScript [
   &ParticleState()
   &SkeletalMeshLocation(
    Bone Sampling Mode=::NewEnumerator0
    "Interpolate Mesh Sampling (Update)"=true
    MaskSpawnGroup=false
    OverrideRandomSeed=false
    Sampled Orientation Coordinate Space=::Local
    Sampled Position Offset=<20 -3 -5>
    Skeletal Mesh=DataInterface() {propertyValues {
      DefaultMesh=null
      ExcludeBoneName=root
      OptionalComponentClass=null
      PreviewMesh=&M_MED_Crisis
      SkinningMode=SkinOnTheFly
      SoftSourceActor=null
      SourceMode=Default
      UvSetIndex=0
      WholeMeshLOD=-1
      bExcludeBone=false
      bReadDeformedGeometry=true
      bRequireCurrentFrameData=true
      ComponentTags []
      FilteredBones [thigh_r]
      FilteredSockets []
      MeshUserParameter {parameter {name=None, typeDefHandle {registeredTypeIndex=1}, varData []}}
      SamplingRegions []
     }}
   )
   &ScaleColor(
    Scale Alpha=&FloatFromCurve(FloatCurve=DataInterface() {propertyValues {Curve {keys [
         {arriveTangent=4.237, interpMode=RCIM_Cubic, leaveTangent=4.237, tangentMode=RCTM_User, time=0, value=0}
         {arriveTangent=0.1123, interpMode=RCIM_Cubic, leaveTangent=0.1123, tangentMode=RCTM_User, time=0.5, value=1}
         {arriveTangent=-4.5416, interpMode=RCIM_Cubic, leaveTangent=-4.5416, tangentMode=RCTM_User, time=1, value=0}
        ]}}})
    ScaleA=true
    ScaleRGB=true
   )
   &FN_Light_Attributes(
    Color=<rgba 1 0.1307 0.7967 1>
    Create Color Attribute=true
    Create Exponent Attribute=true
    Create Position Attribute=true
    Create Radius Attribute=true
    Exponent=1
    Position=&Add_Vector(A=$StackContext.SkeletalMeshLocation.SampledPosition, B=<5 0 -24>)
    Radius=15
    Scale Color=10
   )
  ]
  rendererClasses [&NiagaraMeshRendererProperties, &NiagaraLightRendererProperties]
  renderers [
   {rendererClass=&NiagaraMeshRendererProperties, rendererIndex=0}
   {rendererClass=&NiagaraLightRendererProperties, rendererIndex=1}
  ]
 }
 Lightining {
  bEnabled=false
  emitterUpdateScript [
   &EmitterState()
   &SpawnRate(SpawnRate=8, Use Spawn Probability=false)
  ]
  particleSpawnScript [
   &PartitionParticles(Number Of Partitions=8)
   &InitializeParticle(
    Color=$System.LightningColor
    Lifetime Max=1
    Lifetime Min=0.2
    Lifetime Mode=::NewEnumerator1
    Mesh Renderer Array Visibility Mode=::NewEnumerator1
    Mesh Renderer Info=DataInterface() {propertyValues {MeshRenderer=&self:Lightining_0.NiagaraMeshRendererProperties_0}}
    Mesh Renderer Mesh Index:NiagaraInt32=$Particles.PartitionParticles.Partition
    Mesh Scale Mode=::NewEnumerator4
    UsePositionOffset=false
    Write Material Random=true
   )
   &InitialMeshOrientation(
    Mesh Orientation Mode=::NewEnumerator2
    Rotation=&SelectVectorFromArray(
     Direct Array Index=$Particles.PartitionParticles.Partition
     Vector Selection Array=DataInterface() {propertyValues {FloatData [<0 -0.25 0>, <0 -0.25 0>, <0 -0.24 0>, <0 -0.2 0>, <0 -0.25 0>, <0 0.25 0>, <-0.05 0.25 0>, <-0.05 0.2 0>]}}
    )
    Rotation Coordinate Space=::NewEnumerator2
    Use Rotation Vector=true
   )
   &DynamicMaterialParameters(
    Index 0 Param 1=3
    Index 0 Param 2=1.1
    Index 0 Param 3=0.55
    Index 0 Param 4=1.4
    Module.Param0WriteEnabled=true
    Module.Param1WriteEnabled=true
    Module.Param2WriteEnabled=true
    Module.Param3WriteEnabled=true
   )
  ]
  particleUpdateScript [
   &ParticleState()
   &SkeletalMeshLocation(
    Bone / Socket Index=$Particles.PartitionParticles.Partition
    Bone Sampling Mode=::NewEnumerator4
    "Interpolate Mesh Sampling (Update)"=true
    MaskSpawnGroup=false
    Sampled Orientation Coordinate Space=::Local
    Sampled Position Offset=&SelectVectorFromArray(
     Direct Array Index=$Particles.PartitionParticles.Partition
     Vector Selection Array=DataInterface() {propertyValues {FloatData [<9 2 4>, <5 2 1>, <26 -3 2>, <20 -3 -5>, <-18 -6 0>, <-20 0 -4>, <-10 0 3>, <-6 1 -3>]}}
    )
    Skeletal Mesh=$System.Skeletal Mesh Reader
   )
   &ScaleColor(
    Scale Alpha=&FloatFromCurve(
     FloatCurve=DataInterface() {propertyValues {Curve {keys [
         {arriveTangent=4.237, interpMode=RCIM_Cubic, leaveTangent=4.237, tangentMode=RCTM_User, time=0, value=0}
         {arriveTangent=0.1123, interpMode=RCIM_Cubic, leaveTangent=0.1123, tangentMode=RCTM_User, time=0.5, value=1}
         {arriveTangent=-4.5416, interpMode=RCIM_Cubic, leaveTangent=-4.5416, tangentMode=RCTM_User, time=1, value=0}
        ]}}}
     Scale Curve=&MergeHelper_FloatSwitch(
      0=1
      1=&MakeCustomFloatFromBool(Bool=$User.bisHandHidden, True Float=0, False Float=1)
      2=&MakeCustomFloatFromBool(Bool=$User.bisFootHidden, True Float=0, False Float=1)
      3=1
      4=&MakeCustomFloatFromBool(Bool=$User.bisFootHidden, True Float=0, False Float=1)
      5=1
      6=&MakeCustomFloatFromBool(Bool=$User.bisHandHidden, True Float=0, False Float=1)
      7=1
      Index=$Particles.PartitionParticles.Partition
      Num=::NewEnumerator6
     )
    )
    ScaleA=true
    ScaleRGB=true
   )
   &FN_Light_Attributes(
    Color=$System.LightColor
    Create Color Attribute=true
    Create Exponent Attribute=true
    Create Position Attribute=true
    Create Radius Attribute=true
    Exponent=1
    Position=&Add_Vector(
     A=$StackContext.SkeletalMeshLocation.SampledPosition
     B=&SelectVectorFromArray(
      Direct Array Index=$Particles.PartitionParticles.Partition
      Vector Selection Array=DataInterface() {propertyValues {FloatData [<15 6 -7>, <-10 9 0>, <-5 0 -20>, <5 0 -24>, <5 0 -24>, <15 0 -20>, <0 0 0>, <10 10 0>]}}
     )
    )
    Radius=12
    Scale Color=6
   )
  ]
  rendererClasses [&NiagaraMeshRendererProperties, &NiagaraLightRendererProperties]
  renderers [
   {rendererClass=&NiagaraMeshRendererProperties, rendererIndex=0}
   {rendererClass=&NiagaraLightRendererProperties, rendererIndex=1}
  ]
 }
 Lightining001 {
  emitterUpdateScript [
   &EmitterState()
   &SpawnRate(SpawnRate=8, Use Spawn Probability=false)
  ]
  particleSpawnScript [
   &PartitionParticles(Number Of Partitions=8)
   &InitializeParticle(
    Color=$System.LightningColor
    Lifetime Max=1
    Lifetime Min=0.2
    Lifetime Mode=::NewEnumerator1
    Mesh Renderer Array Visibility Mode=::NewEnumerator1
    Mesh Renderer Info=DataInterface() {propertyValues {MeshRenderer=&self:Lightining001_0.NiagaraMeshRendererProperties_0}}
    Mesh Renderer Mesh Index:NiagaraInt32=$Particles.PartitionParticles.Partition
    Mesh Scale Mode=::NewEnumerator4
    UsePositionOffset=false
    Write Material Random=true
   )
   &InitialMeshOrientation(
    Mesh Orientation Mode=::NewEnumerator2
    Rotation=&SelectVectorFromArray(
     Direct Array Index=$Particles.PartitionParticles.Partition
     Vector Selection Array=DataInterface() {propertyValues {FloatData [<0 -0.25 0>, <0 -0.25 0>, <0 -0.24 0>, <0 -0.2 0>, <0 -0.25 0>, <0 0.25 0>, <-0.05 0.25 0>, <-0.05 0.2 0>]}}
    )
    Rotation Coordinate Space=::NewEnumerator2
    Use Rotation Vector=true
   )
   &DynamicMaterialParameters(
    Index 0 Param 1=0.4
    Index 0 Param 2=1.1
    Index 0 Param 3=0.1
    Index 0 Param 4=&FloatFromCurve(FloatCurve=DataInterface() {propertyValues {Curve {
        defaultValue=1
        keys [{interpMode=RCIM_Cubic, tangentMode=RCTM_User, time=0, value=1}, {interpMode=RCIM_Cubic, tangentMode=RCTM_User, time=1, value=0}]
       }}})
    Index 1 Param 1=&FloatFromCurve(FloatCurve=DataInterface() {propertyValues {Curve {
        defaultValue=0
        keys [
         {interpMode=RCIM_Cubic, tangentMode=RCTM_Break, time=0, value=0}
         {arriveTangent=3.393, interpMode=RCIM_Cubic, tangentMode=RCTM_Break, time=1, value=1}
        ]
       }}})
    Index 1 Param 2=0.5
    Index 1 Param 3=0.5
    Index 1 Param 4=1
    Module.Param0WriteEnabled=true
    Module.Param0WriteEnabled1=true
    Module.Param1WriteEnabled=true
    Module.Param1WriteEnabled1=true
    Module.Param2WriteEnabled=true
    Module.Param2WriteEnabled1=true
    Module.Param3WriteEnabled=true
    Module.Param3WriteEnabled1=true
    Write Parameter Index 1=true
   )
  ]
  particleUpdateScript [
   &ParticleState()
   &SkeletalMeshLocation(
    Bone / Socket Index=$Particles.PartitionParticles.Partition
    Bone Sampling Mode=::NewEnumerator4
    "Interpolate Mesh Sampling (Update)"=true
    MaskSpawnGroup=false
    Sampled Orientation Coordinate Space=::Local
    Sampled Position Offset=&SelectVectorFromArray(
     Direct Array Index=$Particles.PartitionParticles.Partition
     Vector Selection Array=DataInterface() {propertyValues {FloatData [<9 2 4>, <5 2 1>, <26 -3 2>, <20 -3 -5>, <-18 -6 0>, <-20 0 -4>, <-10 0 3>, <-6 1 -3>]}}
    )
    Skeletal Mesh=$System.Skeletal Mesh Reader
   )
   &ScaleColor(
    Scale Alpha=&FloatFromCurve(
     FloatCurve=DataInterface() {propertyValues {Curve {keys [
         {arriveTangent=4.237, interpMode=RCIM_Cubic, leaveTangent=4.237, tangentMode=RCTM_User, time=0, value=0}
         {arriveTangent=0.1123, interpMode=RCIM_Cubic, leaveTangent=0.1123, tangentMode=RCTM_User, time=0.5, value=1}
         {arriveTangent=-4.5416, interpMode=RCIM_Cubic, leaveTangent=-4.5416, tangentMode=RCTM_User, time=1, value=0}
        ]}}}
     Scale Curve=&MergeHelper_FloatSwitch(
      0=1
      1=&MakeCustomFloatFromBool(Bool=$User.bisHandHidden, True Float=0, False Float=1)
      2=&MakeCustomFloatFromBool(Bool=$User.bisFootHidden, True Float=0, False Float=1)
      3=1
      4=&MakeCustomFloatFromBool(Bool=$User.bisFootHidden, True Float=0, False Float=1)
      5=1
      6=&MakeCustomFloatFromBool(Bool=$User.bisHandHidden, True Float=0, False Float=1)
      7=1
      Index=$Particles.PartitionParticles.Partition
      Num=::NewEnumerator6
     )
    )
    ScaleA=true
    ScaleRGB=true
   )
   &FN_Light_Attributes(
    Color=$System.LightColor
    Create Color Attribute=true
    Create Exponent Attribute=true
    Create Position Attribute=true
    Create Radius Attribute=true
    Exponent=1
    Position=&Add_Vector(
     A=$StackContext.SkeletalMeshLocation.SampledPosition
     B=&SelectVectorFromArray(
      Direct Array Index=$Particles.PartitionParticles.Partition
      Vector Selection Array=DataInterface() {propertyValues {FloatData [<15 6 -7>, <-10 9 0>, <-5 0 -20>, <5 0 -24>, <5 0 -24>, <15 0 -20>, <0 0 0>, <10 10 0>]}}
     )
    )
    Radius=12
    Scale Color=6
   )
  ]
  rendererClasses [&NiagaraMeshRendererProperties, &NiagaraLightRendererProperties]
  renderers [
   {rendererClass=&NiagaraMeshRendererProperties, rendererIndex=0}
   {rendererClass=&NiagaraLightRendererProperties, rendererIndex=1}
  ]
 }
 Lines {
  emitterUpdateScript [
   &EmitterState(Life Cycle Mode=::NewEnumerator1, Loop Duration=2, UseLoopDelay=false)
   &SpawnRate(SpawnRate=&Multiply_Float(A=25, B=$User.Variant), Use Spawn Probability=false)
  ]
  particleSpawnScript [
   &InitializeParticle(
    Color Maximum=<rgba 0.4614 0 1 0.65>
    Color Minimum=<rgba 1 0 0.9692 0.65>
    Color Mode=::NewEnumerator2
    Lifetime Max=0.6
    Lifetime Min=0.4
    Lifetime Mode=::NewEnumerator1
    Mass Mode=::NewEnumerator3
    Sprite Rotation Mode=::NewEnumerator2
    Sprite Size Max=<1 12>
    Sprite Size Min=<0.5 10>
    Sprite Size Mode=::NewEnumerator2
    UsePositionOffset=false
    Write Material Random=true
   )
   &TorusLocation(Align Torus With Axis=true, Handle Radius=1, Large Radius=15, MaskSpawnGroup=false, Module.Override Randomness=false, Torus Axis=<0 1 0>)
   &CylinderLocation(
    Cylinder Height=3
    Cylinder Radius=15
    MaskSpawnGroup=false
    Override Seed=false
    Surface Only=true
    Use Endcaps In Surface Only Mode=false
    UseOwnerScale=true
   ) {enabled=false}
   &AddVelocityFromPoint(Module.Use Falloff Distance=false, Velocity Strength=25)
  ]
  particleUpdateScript [
   &ParticleState()
   &SkeletalMeshLocation(
    Bone Sampling Mode=::NewEnumerator4
    MaskSpawnGroup=false
    Position Sampling=::NewEnumerator1
    Sampled Position Offset=<12 0 0>
    Skeletal Mesh=$User.New Skeletal Mesh
   )
   &ScaleColor(
    Scale Alpha=&FloatFromCurve(FloatCurve=DataInterface() {propertyValues {Curve {keys [
         {arriveTangent=-1.3337, interpMode=RCIM_Cubic, leaveTangent=-1.3337, tangentMode=RCTM_User, time=-0.0046, value=1.0113}
         <at 1.0011 0.002 RCIM_Cubic>
        ]}}})
    ScaleA=true
    ScaleRGB=true
   )
   &SolveForcesAndVelocity(Clamp Velocity=false, Limit Acceleration=false)
   &ScaleSpriteSize(Scale Factor=&Vector2DFromCurve(
     Vector2Curve=DataInterface() {propertyValues {XCurve {keys [<at 0 0.5 RCIM_Cubic>,<at 0.5 1 RCIM_Cubic>]}, YCurve {keys [<at 0 0 RCIM_Cubic>,<at 0.5 1 RCIM_Cubic>]}}}
     ScaleCurve=&Vector2DFromFloat(Value=1)
    ))
   /SetVariables(Particles.VisibilityTag=&MakeCustomIntFromBool(Bool=$User.bisHeadHidden, True INT=1, False INT=0)) {isSetParameters=true, module=SetVariables_#3}
  ]
  rendererClasses [&NiagaraSpriteRendererProperties]
  renderers [{rendererClass=&NiagaraSpriteRendererProperties}]
 }
}
summary {
 systemName=NS_MED_Crisis
 emitters [
  {
   bEnabled=false
   emitterName=Hand_L1
   rendererClasses [&NiagaraMeshRendererProperties, &NiagaraLightRendererProperties]
  }
  {
   bEnabled=false
   emitterName=Hand_L2
   rendererClasses [&NiagaraMeshRendererProperties, &NiagaraLightRendererProperties]
  }
  {
   bEnabled=false
   emitterName=Leg_R
   rendererClasses [&NiagaraMeshRendererProperties, &NiagaraLightRendererProperties]
  }
  {
   bEnabled=false
   emitterName=Leg_R1
   rendererClasses [&NiagaraMeshRendererProperties, &NiagaraLightRendererProperties]
  }
  {
   bEnabled=false
   emitterName=Leg_L1
   rendererClasses [&NiagaraMeshRendererProperties, &NiagaraLightRendererProperties]
  }
  {
   bEnabled=false
   emitterName=Leg_L2
   rendererClasses [&NiagaraMeshRendererProperties, &NiagaraLightRendererProperties]
  }
  {bEnabled=false, emitterName=Hand_R1001, rendererClasses [&NiagaraMeshRendererProperties]}
  {
   bEnabled=false
   emitterName=Hand_L1001
   rendererClasses [&NiagaraMeshRendererProperties, &NiagaraLightRendererProperties]
  }
  {emitterName=HeadToShoulder, rendererClasses [&NiagaraMeshRendererProperties]}
  {emitterName=HeadToShoulderWide, rendererClasses [&NiagaraMeshRendererProperties]}
  {emitterName=HeadUp, rendererClasses [&NiagaraMeshRendererProperties]}
  {emitterName=Energy, rendererClasses [&NiagaraSpriteRendererProperties, &NiagaraLightRendererProperties]}
  {emitterName=Lines, rendererClasses [&NiagaraSpriteRendererProperties]}
  {
   bEnabled=false
   emitterName=Lightining
   rendererClasses [&NiagaraMeshRendererProperties, &NiagaraLightRendererProperties]
  }
  {emitterName=Lightining001, rendererClasses [&NiagaraMeshRendererProperties, &NiagaraLightRendererProperties]}
 ]
 userVariables [
  User.bisFootHidden {type=NiagaraBool}
  User.bisHandHidden {type=NiagaraBool}
  User.bisHeadHidden {type=NiagaraBool}
  User.ColorVariant {type=NiagaraFloat, defaultValue=0, description=""}
  User.New Skeletal Mesh {
   type=NiagaraDataInterfaceSkeletalMesh
   defaultValue {
    struct=NiagaraExt_VariableValue_DataInterface
    value {dataInterface=&self:NiagaraDataInterfaceSkeletalMesh_5}
   }
  }
  User.Variant {type=NiagaraFloat, defaultValue=0, description=""}
 ]
}
