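# Master character material for the Crisis skin, layering animated energy emissive and optional cube FX between the shared character PreFX, SkinFX and PostFX functions.
# A timing mask, distorted by curl and crystal noise, drives sweeping surface energy plus looping rune and collar/face glows.
# UseCubeFX switches in Evolved-variant cubes with curve-atlas pulses along vertex normals, view-angle color shift and HDR glows.
# UseHeadCubeFX selects head-specific cube timing and glow.
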Material /BRCosmetics/Characters/Player/Male/Medium/Bodies/M_MED_Crisis/Materials/M_FN_Character_Crisis_MASTER
graph {
 properties {
  bAutomaticallySetUsageInEditor=false
  bUseMaterialAttributes=true
  bUsedWithMeshDeformer=true
  bUsedWithMorphTargets=true
  bUsedWithSkeletalMesh=true
  pixelDepthOffsetMode=PDOM_Legacy
  refractionMethod=RM_IndexOfRefraction
  shadingModel=MSM_FromMaterialExpression
  subsurfaceProfile=&SS_HeroSkin_02
  usedShadingModels=Default Lit | Subsurface | Clear Coat | Subsurface Profile | Cloth | Eye
 }
 nodes [
  n1 = MaterialFunctionCall() {materialFunction=&BaseFN_CharacterFunction_PreFX}
  n2 = GetMaterialAttributes(None=MaterialFunctionCall(
    PreskinnedPosition=@n1.Pre skinned Local Position
    Gmap=@n1.Pre skinned Local Position
    MaterialAttributes=@n1.Pre skinned Local Position
    AO=@n1.Pre skinned Local Position
   ) {materialFunction=&CharacterSkinFX}.Result) {attributeGetTypes [#1]}
  n3 = TextureObjectParameter() {
   group=Skin FX
   parameterName=SkinFX_Mask
   texture=/BRCosmetics/Characters/Player/Male/Medium/Bodies/M_MED_Crisis/Textures/T_M_Med_Crisis_Body_FX
  }
  n4 = TextureSample(Tex=@n3) {
   samplerSource=SSM_Wrap_WorldGroupSettings
   samplerType=SAMPLERTYPE_Masks
   texture=/BRCosmetics/Characters/Player/Male/Medium/Bodies/M_MED_Crisis/Textures/T_M_Med_Crisis_Body_FX_Timing
  }
  n5 = StaticSwitchParameter(True=Constant() {r=2}, False=Multiply(A=@n4.B) {constB=2}) {parameterName=UseCubeFX}
  n6 = TextureCoordinate()
  n7 = Multiply(A=TextureCoordinate(), B=ScalarParameter() {defaultValue=1, group=CrisisEmissiveBreakup, parameterName=CurlTile})
  n8 = Multiply(
   A=ComponentMask(None=TextureSample(UVs=@n7) {samplerType=SAMPLERTYPE_LinearColor, texture=/Game/TimeOfDay/VolumeFog/T_CurlNoise}.RGB) {g=true, r=true}
   B=ScalarParameter() {defaultValue=0.001, group=CrisisEmissiveBreakup, parameterName=CurlStrength}
  )
  n9 = TextureSample(UVs=Multiply(A=TextureCoordinate()) {constB=1.5}) {
   samplerSource=SSM_Wrap_WorldGroupSettings
   samplerType=SAMPLERTYPE_Masks
   texture=/Game/Accessories/FORT_Backpacks/M_MED_Heist/Textures/T_CrystalNoise_M
  }
  n10 = MaterialFunctionCall(X=@n9.G, Y=@n9.G) {materialFunction=&MakeFloat2}
  n11 = ScalarParameter() {defaultValue=0.03, group=CrisisEmissiveBreakup, parameterName=CurlBreakupStrength}
  n12 = TextureSampleParameter2D(UVs=Add(A=Add(A=@n6, B=@n8), B=Multiply(A=@n10.Result, B=@n11))) {
   parameterName=FX_Timing
   samplerType=SAMPLERTYPE_Masks
   texture=&T_M_Med_Crisis_Body_FX_Timing
  }
  n13 = Add(A=@n12.R, B=Multiply(A=Time(), B=ScalarParameter() {defaultValue=0.5, group=CrisisEmissiveBreakup, parameterName=SurfaceEnergySpeed}))
  n14 = Power(Base=OneMinus(None=Frac(None=@n13))) {constExponent=60}
  n15 = ScalarParameter() {defaultValue=10, group=CrisisEmissiveBreakup, parameterName=RuneEmissiveIntensity}
  n16 = TextureSampleParameter2D() {
   parameterName=FX_Timing
   samplerType=SAMPLERTYPE_Masks
   texture=&T_M_Med_Crisis_Body_FX_Timing
  }
  n17 = Add(A=Multiply(A=Time(), B=ScalarParameter() {defaultValue=0.15, group=CrisisEmissiveBreakup, parameterName=GlowLoopSpeed}), B=@n16.G)
  n18 = CurveAtlasRowParameter(CurveTime=Frac(None=@n17)) {
   atlas=/BRCosmetics/Characters/Player/Male/Medium/Bodies/M_MED_Crisis_Evolved/Materials/CA_M_Med_Crisis_Evolved_CubeTiming
   curve=/BRCosmetics/Characters/Player/Male/Medium/Bodies/M_MED_Crisis_Evolved/Materials/Curve_Crisis_Evolved_Cube_Timing
   parameterName=CubeTimingCurve
  }
  n19 = Multiply(A=@n16.B, B=@n18.R)
  n20 = MaterialFunctionCall(Value=@n17) {materialFunction=&LinearSine}
  n21 = Power(Base=@n20.Linear Sine) {constExponent=10}
  n22 = Multiply(A=Max(A=Saturate(None=@n21)) {constB=0}, B=@n15)
  n23 = Multiply(A=Max(A=@n20.Linear Sine) {constB=0.15}, B=ScalarParameter() {defaultValue=2, group=CrisisEmissiveBreakup, parameterName=CollarFaceGlow})
  n24 = Add(
   A=Multiply(A=Saturate(None=@n14), B=ScalarParameter() {defaultValue=6, group=CrisisEmissiveBreakup, parameterName=SurfaceEngergyEmissiveIntensity})
   B=StaticSwitchParameter(True=Multiply(A=@n15, B=@n19), False=LinearInterpolate(A=@n22, B=@n23, Alpha=@n16.B)) {parameterName=UseCubeFX}
  )
  n25 = SetMaterialAttributes(MaterialAttributes=@n2.MaterialAttributes, Emissive Color=Multiply(A=@n2.Emissive Color, B=Multiply(A=@n5, B=@n24))) {attributeSetTypes [#1]}
  n26 = GetMaterialAttributes(None=@n25) {attributeGetTypes [#2]}
  n27 = StaticSwitchParameter(True=NamedRerouteDeclaration(None=@n26.Base Color) {name=BaseColor, nodeColor=<rgba 0.4766 1 0 1>}, False=NamedRerouteUsage()) {parameterName=UseHeadCubeFX}
  n28 = Add(A=GetMaterialAttributes(None=@n25) {attributeGetTypes [#1]}.Emissive Color, B=NamedRerouteUsage())
  n29 = Add(A=GetMaterialAttributes(None=@n25) {attributeGetTypes [#3]}.World Position Offset, B=NamedRerouteUsage())
  n30 = MaterialFunctionCall(
   AO=@n1.AO
   Distance Blend=@n1.AO
   Fresnel=@n1.AO
   Gmap=@n1.AO
   Pre Skinned Local Position=@n1.AO
   Material=StaticSwitchParameter(
    True=SetMaterialAttributes!(MaterialAttributes=@n26.MaterialAttributes, Base Color=@n27, Emissive Color=@n28, World Position Offset=@n29) {attributeSetTypes [#2, #1, #3]}
    False=@n25
   ) {parameterName=UseCubeFX}
  ) {materialFunction=&BaseFN_CharacterFunction_PostFX}
  MaterialFunctionCall(In=@n30.Result) {materialFunction=&CINE_MF_CharPostFX}
  n31 = VertexNormalWS()
  n32 = ScalarParameter() {defaultValue=0.66, parameterName=CubePulseAmountWS}
  n33 = Multiply(A=Time(), B=ScalarParameter() {defaultValue=0.5, parameterName=CubePulseSpeed})
  n34 = Add(A=@n33, B=ComponentMask(None=TextureCoordinate()) {g=true})
  n35 = CurveAtlasRowParameter(CurveTime=Frac(None=StaticSwitchParameter(True=@n33, False=@n34) {parameterName=UseHeadCubeFX})) {
   atlas=&CA_M_Med_Crisis_Evolved_CubeTiming
   curve=&Curve_Crisis_Evolved_Cube_Timing
   parameterName=CubeTimingCurve
  }
  NamedRerouteDeclaration(None=Multiply(A=Multiply(A=@n31, B=@n32), B=@n35.R)) {name=CubeFXWPO, nodeColor=<rgba 0.8516 0 1 1>}
  n36 = NamedRerouteUsage()
  n37 = Constant3Vector() {constant=<rgba 0.0721 0 0.1823 0>}
  n38 = TextureSample() {
   samplerSource=SSM_Wrap_WorldGroupSettings
   samplerType=SAMPLERTYPE_Masks
   texture=/BRCosmetics/Characters/Player/Male/Medium/Bodies/M_MED_Crisis_Evolved/Textures/T_M_MED_Crisis_Evolved_Cubes_MaterialFX
  }
  n39 = Saturate(None=Multiply(A=OneMinus(None=@n38.R)) {constB=1.2})
  n40 = Power(Base=DotProduct(A=CameraVectorWS(), B=VertexNormalWS()), Exp=ScalarParameter() {defaultValue=5, parameterName=CubeGlowFalloffExp})
  n41 = CurveAtlasRowParameter(CurveTime=NamedRerouteDeclaration(None=Saturate(None=@n40)) {name=CubeEdgeFalloffB, nodeColor=<rgba 0 1 0.8594 1>}) {
   atlas=&CA_M_Med_Crisis_Evolved_CubeTiming
   curve=/BRCosmetics/Characters/Player/Male/Medium/Bodies/M_MED_Crisis_Evolved/Materials/Curve_Crisis_Evolved_Cube_Colors
   defaultValue=1
   parameterName=CubeColorCurve
  }
  n42 = DotProduct(A=CameraVectorWS(), B=PixelNormalWS())
  n43 = ScalarParameter() {defaultValue=1, parameterName=CubeHueShiftFresnelExp}
  n44 = NamedRerouteDeclaration(None=Saturate(None=Power(Base=@n42, Exp=@n43) {constExponent=5})) {name=CubeEdgeFalloffA, nodeColor=<rgba 1 0 0.0703 1>}
  n45 = Saturate(None=Power(Base=OneMinus(None=@n44)) {constExponent=0.3})
  NamedRerouteDeclaration(None=LinearInterpolate(A=LinearInterpolate(A=@n36, B=@n37, Alpha=@n39), B=@n41, Alpha=Multiply(A=@n45, B=@n39))) {name=ModifiedBaseColor, nodeColor=<rgba 1 0 0.5391 1>}
  n46 = Multiply(A=@n38.B, B=NamedRerouteUsage())
  n47 = StaticSwitchParameter(True=NamedRerouteUsage(), False=NamedRerouteUsage()) {parameterName=UseHeadCubeFX}
  n48 = MaterialFunctionCall(
   LDR Input=Multiply(A=Multiply(A=@n46, B=@n47), B=NamedRerouteUsage())
   HDR Intensity Multiplier=ScalarParameter() {defaultValue=8, parameterName=LargeCubeEdgeGlow}
  ) {materialFunction=&DeriveHDRfromLDR}
  n49 = NamedRerouteUsage()
  n50 = Saturate(None=Power(Base=OneMinus(None=@n49)) {constExponent=30})
  n51 = Multiply(A=Add(A=@n38.R, B=Multiply(A=@n38.G, B=@n50)), B=NamedRerouteUsage())
  n52 = NamedRerouteUsage()
  n53 = NamedRerouteUsage()
  n54 = NamedRerouteUsage()
  n55 = MaterialFunctionCall(
   LDR Input=Multiply(A=Multiply(A=@n51, B=@n52), B=StaticSwitchParameter(True=@n53, False=@n54) {parameterName=UseHeadCubeFX})
   HDR Intensity Multiplier=ScalarParameter() {defaultValue=1, parameterName=InnerGlow}
  ) {materialFunction=&DeriveHDRfromLDR}
  NamedRerouteDeclaration(None=StaticSwitchParameter(True=Add(A=@n55.RemappedRange, B=@n48.RemappedRange), False=@n55.RemappedRange) {parameterName=UseHeadCubeFX}) {name=ModifiedEmissive, nodeColor=<rgba 0 0.9062 1 1>}
  n56 = Add(
   A=ComponentMask(None=TextureCoordinate()) {g=true}
   B=Multiply(A=Time(), B=ScalarParameter() {defaultValue=0.4, group=CrisisEmissiveBreakup, parameterName=CubeInteriorGlowSpeed})
  )
  NamedRerouteDeclaration(None=LinearInterpolate(Alpha=MaterialFunctionCall(Value=@n56) {materialFunction=&LinearSine}."Rounded Linear Sine ") {constA=0.1}) {name=CubeGlowSinusoidalIntensity, nodeColor=<rgba 1 0 0.8672 1>}
 ]
}
