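# Animation Blueprint for the M_MED_Crisis player body cosmetic, copying the attached parent mesh's pose and adding corrective deformation.
# The copied pose runs through twist and sculpter Control Rigs, then additive offsets on both front elbow deform bones.
# The event graph drives those offsets' alphas from each lowerarm's yaw delta from reference pose, relative to its upperarm.
# Finally, head_offset is constrained to spine_05 and partially aimed at head_offset_AIM.
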AnimBlueprint /BRCosmetics/Characters/Player/Male/Medium/Bodies/M_MED_Crisis/Meshes/M_MED_Crisis_AnimBP
target_skeleton=&/M_MED_Crisis_Skeleton
graphs [
 AnimGraph {nodes [
   n1 = Control Rig(Source=Copy Pose From Mesh() {bUseAttachedParent=true}) {
    alphaBoolBlend {blendOption=Linear}
    controlRigAssetReference {blueprintRigClass=&/M_MED_Crisis_Deform_CtrlRig_twists.M_MED_Crisis_Deform_CtrlRig_twists_C}
   }
   n2 = Control Rig!(
    Source=@n1
    Alpha
    KneeRot_fr_l_114deg:Vector
    KneeTrans_fr_l_114deg:Vector="0.258,-3.591,0.009"
    KneeScale_fr_l_114deg:Vector="1,1,1"
    KneeScale_fr_l_130deg:Vector="1,1,1"
    KneeRot_fr_l_130deg:Vector="0,0,-51.2"
    KneeTrans_fr_l_130deg:Vector="0.258,-3.591,0.009"
   ) -> (Pose) {
    alphaBoolBlend {blendOption=Linear}
    controlRigAssetReference {blueprintRigClass=&/M_MED_Crisis_Deform_CtrlRig_sculpters.M_MED_Crisis_Deform_CtrlRig_sculpters_C}
   }
   n3 = "Transform (Modify) Bone - Bone: deform_elbow_fr_l"(
    ComponentPose=Local To Component(LocalPose=@n2)
    Alpha=Get elbow_rot_l() -> (elbow_rot_l:double)
    Translation="10,15,0"
    Rotation="0,0,0"
    Scale="1,1,1"
   ) {
    rotationMode=BMM_Additive
    rotationSpace=BCS_BoneSpace
    scaleMode=BMM_Additive
    scaleSpace=BCS_BoneSpace
    translationMode=BMM_Additive
    translationSpace=BCS_BoneSpace
    alphaBoolBlend {blendOption=Linear}
    boneToModify {boneName=deform_elbow_fr_l}
    translation=<10 15>
   }
   n4 = "Transform (Modify) Bone - Bone: deform_elbow_fr_r"(ComponentPose=@n3, Alpha=Get elbow_rot_r() -> (elbow_rot_r:double), Translation="10,15,0", Rotation="0,0,0", Scale="1,1,1") {
    rotationMode=BMM_Additive
    rotationSpace=BCS_BoneSpace
    scaleMode=BMM_Additive
    scaleSpace=BCS_BoneSpace
    translationMode=BMM_Additive
    translationSpace=BCS_BoneSpace
    alphaBoolBlend {blendOption=Linear}
    boneToModify {boneName=deform_elbow_fr_r}
    translation=<10 15>
   }
   n5 = "Constraint - head_offset (P)"(ComponentPose=@n4, Alpha=1, ConstraintWeights_0=1) {
    alphaBoolBlend {blendOption=Linear}
    boneToModify {boneName=head_offset}
    constraintSetup [{targetBone {boneName=spine_05}}]
    constraintWeights [1]
   }
   n6 = Look At - Bone: head_offset(ComponentPose=@n5, Alpha=0.4) {
    alphaBoolBlend {blendOption=Linear}
    boneToModify {boneName=head_offset}
    lookAtLocation {x=0}
    lookAtTarget {boneReference {boneName=head_offset_AIM}}
    lookAt_Axis {axis=<1 0>}
   }
   Output Pose(Result=Component To Local(ComponentPose=@n6)) {layerGroup=None}
  ]}
 EventGraph {nodes [
   Event BlueprintUpdateAnimation()
   > n1 = Sequence()
   n2 = BreakTransform(InTransform=GetDeltaTransformFromRefPose(self=GetOwnerSkeletalMeshComponent(), BoneName=lowerarm_l, BaseName=upperarm_l))
   > Set elbow_rot_l(elbow_rot_l:double=MapRangeClamped(Value=BreakRotator(InRot=@n2.Rotation).Yaw, InRangeA=0, InRangeB=110)) -> (Output_Get:double)
   n3 = BreakTransform(InTransform=GetDeltaTransformFromRefPose(self=GetOwnerSkeletalMeshComponent(), BoneName=lowerarm_r, BaseName=upperarm_r))
   Set elbow_rot_r(execute=@n1.then_1, elbow_rot_r:double=MapRangeClamped(Value=BreakRotator(InRot=@n3.Rotation).Yaw, InRangeA=0, InRangeB=110)) -> (Output_Get:double)
  ]}
]
variables [
 elbow_rot_l {type=number}
 elbow_rot_r {type=number}
]
